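# Niagara impact effect for the ClaimReflect pickaxe, a short burst of flares, flash, smoke, streaks and sparks.
# System.Variant switches most emitter colors between blue and red, while System.DirectionVector steers the directional flare, streaks and sparks.
# Flash_Elements drives spike, star, flare and light from one emitter using custom attributes and separate age curves.
# Front-end and time-of-day modifiers scale size, brightness and spawning; the Shockwave mesh emitter is disabled.
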NiagaraSystem /BRCosmetics/Weapons/FORT_Melee/Pickaxe_ClaimReflect/FX/NS_Pickaxe_ClaimReflect_Impact
dependencies {
 usedDataInterfaces [&NiagaraDataInterfaceCurve, &NiagaraDataInterfaceColorCurve, &NiagaraDataInterfaceVector2DCurve]
 usedDynamicInputs [
  &Vector4FromCurve
  &UniformRangedVector2D
  &MakeIntFromFloat
  /Niagara/DynamicInputs/Math/Clamp_Vector
  &FN_FrontEndVector3Modifier
  &MergeHelper_VectorSwitch
 ]
 usedModules [
  &self:SystemSpawnScript.SystemScriptSource.SystemScriptGraph.NiagaraNodeAssignment_1.SetVariables_#1
  &SystemState
  &self:SystemSpawnScript.SystemScriptSource.SystemScriptGraph.NiagaraNodeAssignment_0.SetVariables_#2
 ]
}
emitters {
 ChromaRadialFlare {
  emitterUpdateScript [
   &EmitterState(Life Cycle Mode=::NewEnumerator1, Loop Behavior=::NewEnumerator1, Loop Duration=1, UseLoopDelay=false)
   &SpawnBurst_Instantaneous(Use Spawn Probability=false, UseLoopCountLimit=false) {module=SpawnBurst_Instantaneous002}
  ]
  particleSpawnScript [
   &InitializeParticle(
    Color=&MultiplyLinearColor(
     Linear Color=&MergeHelper_ColorSwitch(0=<rgba 0.4 0.9333 2 1>, 1=<rgba 2 0 0 1>, 2=?, 3=?, Index=$System.Variant, Num=::NewEnumerator0)
     Scale Factor=&MakeVector4_FromLinearColor(LinearColor=$System.TOD)
    )
    Lifetime=0.35
    Sprite Rotation Mode=::NewEnumerator2
    Sprite Size Mode=::NewEnumerator3
    Uniform Sprite Size=&Multiply_Float(A=185, B=&FN_FrontEndFloatModifier(Gameplay=1, Front End=$User.FrontEndScale))
    UsePositionOffset=false
    Write Material Random=true
   ) {module=InitializeParticle001}
   &CameraOffset(Camera Offset Amount=50) {module=CameraOffset001}
  ]
  particleUpdateScript [
   &ParticleState()
   &ScaleColor(
    Scale Alpha=&FloatFromCurve(
     FloatCurve=DataInterface() {propertyValues {Curve {keys [{arriveTangent=-3.7232, interpMode=RCIM_Cubic, leaveTangent=-3.7232, tangentMode=RCTM_User, time=0.2, value=1}, <at 1 0 RCIM_Cubic>]}}}
     Scale Curve=3
    )
    Scale RGB=&MakeVectorFromLinearColorRGB(Color=&MergeHelper_ColorSwitch(
      0=&ColorFromCurve(ColorCurve=DataInterface() {propertyValues {
         AlphaCurve {keys [<at 0 1 RCIM_Cubic>]}
         BlueCurve {keys [<at 0 1 RCIM_Cubic>,<at 0.5354 1>,<at 0.9851 0.6>]}
         GreenCurve {keys [<at 0 0.917 RCIM_Cubic>,<at 0.5354 0.6>,<at 0.9851 0.8395>]}
         RedCurve {keys [<at 0 0.5 RCIM_Cubic>,<at 0.5354 0.7128>,<at 0.9851 1>]}
        }})
      1=&ColorFromCurve(ColorCurve=DataInterface() {propertyValues {
         AlphaCurve {keys [<at 0 1 RCIM_Cubic>]}
         BlueCurve {keys [<at 0 0 RCIM_Cubic>,<at 0.9851 0>]}
         GreenCurve {keys [<at 0 0 RCIM_Cubic>,<at 0.9851 0>]}
         RedCurve {keys [<at 0 1 RCIM_Cubic>,<at 0.9851 1>]}
        }})
      2=?
      3=?
      Index=$System.Variant
      Num=::NewEnumerator0
     ))
    ScaleA=true
    ScaleRGB=true
   )
   &ScaleSpriteSize(Scale Factor=&Vector2DFromFloat(Value=&FloatFromCurve(
      FloatCurve=DataInterface() {propertyValues {Curve {keys [{arriveTangent=1.7733, interpMode=RCIM_Cubic, leaveTangent=1.7733, tangentMode=RCTM_User, time=0, value=0.2}, <at 1 1 RCIM_Cubic>]}}}
      Scale Curve=0.75
     )))
   &DynamicMaterialParameters(
    Index 0 Param 1=&FloatFromCurve(FloatCurve=DataInterface() {propertyValues {Curve {keys [{tangentMode=RCTM_User, time=0, value=0}, {tangentMode=RCTM_User, time=1, value=1}]}}})
    Index 0 Param 2=&FloatFromCurve(
     FloatCurve=DataInterface() {propertyValues {Curve {keys [
         {arriveTangent=2.4836, interpMode=RCIM_Cubic, leaveTangent=2.4836, tangentMode=RCTM_User, time=0, value=0}
         {interpMode=RCIM_Cubic, tangentMode=RCTM_User, time=1, value=1}
        ]}}}
     Scale Curve=0.1
    )
    Index 1 Param 1=3
    Index 1 Param 2=&FloatFromCurve(
     FloatCurve=DataInterface() {propertyValues {Curve {
        defaultValue=1
        keys [
         {arriveTangent=4.1785, interpMode=RCIM_Cubic, leaveTangent=4.1785, tangentMode=RCTM_User, time=0, value=0}
         {arriveTangent=1.1179, interpMode=RCIM_Cubic, leaveTangent=1.1179, tangentMode=RCTM_User, time=0.2042, value=0.6228}
         {arriveTangent=0.0224, interpMode=RCIM_Cubic, leaveTangent=0.0224, tangentMode=RCTM_User, time=1, value=1}
        ]
       }}}
     Scale Curve=4
    )
    Index 1 Param 3=1
    Index 1 Param 4=1
    Module.Index0Param0Write=true
    Module.Index0Param1Write=true
    Module.Index0Param2Write=true
    Module.Index0Param3Write=true
    Module.Index1Param0Write=true
    Module.Index1Param1Write=true
    Module.Index1Param2Write=true
    Module.Index1Param3Write=true
    Write Parameter Index 1=true
   ) {module=DynamicMaterialParameters001}
   /SetVariables(
    Particles.SpikesColor:LinearColor=&MakeLinearColorFromVectorAndFloat(
     "Vector (RGB)"=&MakeVectorFromLinearColorRGB(Color=&MergeHelper_ColorSwitch(
       0=&ColorFromCurve(ColorCurve=DataInterface() {propertyValues {
          AlphaCurve {keys []}
          BlueCurve {keys [<at 0.0027 3>,<at 0.931 0.2 RCIM_Cubic>]}
          GreenCurve {keys [<at 0.0027 0.75>,<at 0.931 0.065 RCIM_Cubic>]}
          RedCurve {keys [<at 0.0027 0.3>,<at 0.931 0.02 RCIM_Cubic>]}
         }})
       1=&ColorFromCurve(ColorCurve=DataInterface() {propertyValues {
          AlphaCurve {keys []}
          BlueCurve {keys [<at 0.0027 0>,<at 0.931 0 RCIM_Cubic>]}
          GreenCurve {keys [<at 0.0027 0>,<at 0.931 0 RCIM_Cubic>]}
          RedCurve {keys [<at 0.0027 3>,<at 0.931 0.2 RCIM_Cubic>]}
         }})
       2=?
       3=?
       Index=$System.Variant
       Num=::NewEnumerator0
      ))
     "Float (Alpha)"=&FloatFromCurve(
      FloatCurve=DataInterface() {propertyValues {Curve {keys [{interpMode=RCIM_Cubic, tangentMode=RCTM_User, time=0, value=1}, {interpMode=RCIM_Cubic, tangentMode=RCTM_User, time=0.6, value=0}]}}}
      Scale Curve=3
     )
    )
    Particles.SpikesSpriteSize:Vector2f=&Multiply_Vector2D_ByFloat(
     Float=&FloatFromCurve(
      FloatCurve=DataInterface() {propertyValues {Curve {keys [{tangentMode=RCTM_User, time=0, value=0.8}, {tangentMode=RCTM_User, time=1, value=1}]}}}
      Scale Curve=&Multiply_Float(
       A=&RandomRangeFloat(Minimum=0.8, Maximum=1)
       B=&FN_FrontEndFloatModifier(Gameplay=0.8, Front End=0.3)
      )
     )
     Vector2D=<200 175>
    )
   ) {isSetParameters=true}
  ]
  rendererClasses [&NiagaraSpriteRendererProperties]
  renderers [
   {rendererClass=&NiagaraSpriteRendererProperties}
   {rendererClass=&NiagaraSpriteRendererProperties, rendererIndex=1}
  ]
 }
 DirectionalFlare {
  emitterUpdateScript [
   &EmitterState(
    Enable Distance Culling=true
    Life Cycle Mode=::NewEnumerator1
    Loop Behavior=::NewEnumerator1
    Loop Delay=0.01
    Loop Duration=1
    Max Distance Response=::KillImmediately
    MaxDistance=3500
    Module.bUseMaxDistance=true
    Module.bUseMinDistance=false
    Scalability Mode=::NewEnumerator1
    UseLoopDelay=false
   )
   &SpawnBurst_Instantaneous(
    Spawn Count=1
    Spawn Probability=&FN_FrontEndFloatModifier(Gameplay=1, Front End=0)
    Spawn Time=0.005
    Use Spawn Probability=true
   )
  ]
  particleSpawnScript [
   &InitializeParticle(
    Color=<rgba 9 9 9 1>
    Lifetime=0.26
    Position Offset=&Multiply_VectorByFloat(Vector=$System.DirectionVector, Float=-100)
    Sprite Size=<17 300>
    Sprite Size Mode=::NewEnumerator4
    UsePositionOffset=true
    Write Material Random=true
   ) {module=InitializeParticle001}
   &CameraOffset(Camera Offset Amount=40)
   &DynamicMaterialParameters(
    Index 0 Param 1=0.2
    Index 0 Param 2=1
    Module.Param0WriteEnabled=true
    Module.Param1WriteEnabled=true
    Module.Param2WriteEnabled=true
    Module.Param3WriteEnabled=true
   )
   &AddVelocity(Velocity=$System.DirectionVector, Velocity Speed Scale=800)
  ]
  particleUpdateScript [
   &ParticleState()
   &SolveForcesAndVelocity(Clamp Velocity=false, Limit Acceleration=false)
   &ScaleColor(
    Color Value To Scale=$Particles.Color
    Scale Alpha=&FloatFromCurve(FloatCurve=DataInterface() {propertyValues {Curve {keys [<at 0.02 0 RCIM_Cubic>,<at 0.06 1 RCIM_Cubic>,<at 1 0 RCIM_Cubic>]}}})
    Scale RGB=&MakeVectorFromLinearColorRGB(Color=&MergeHelper_ColorSwitch(
      0=&ColorFromCurve(ColorCurve=DataInterface() {propertyValues {
         AlphaCurve {keys [<at 0 1 RCIM_Cubic>,<at 1 0 RCIM_Cubic>]}
         BlueCurve {keys [
           <at 0 3>
           {arriveTangent=-3.8268, interpMode=RCIM_Cubic, leaveTangent=-3.8268, time=0.0295, value=2}
           {interpMode=RCIM_Cubic, leaveTangent=-1.3593, time=0.5226, value=1}
          ]}
         GreenCurve {keys [
           <at 0 1.8>
           {arriveTangent=-2.4237, interpMode=RCIM_Cubic, leaveTangent=-2.4237, time=0.0295, value=1.2}
           {interpMode=RCIM_Cubic, leaveTangent=-2.7288, time=0.5226, value=0.5333}
          ]}
         RedCurve {keys [<at 0 0.6>, {arriveTangent=-0.574, interpMode=RCIM_Cubic, leaveTangent=-0.574, time=0.0295, value=0.4}, <at 0.5226 0.3 RCIM_Cubic>]}
        }})
      1=&ColorFromCurve(ColorCurve=DataInterface() {propertyValues {
         AlphaCurve {keys [<at 0 1 RCIM_Cubic>,<at 1 0 RCIM_Cubic>]}
         BlueCurve {keys [
           <at 0 0.6>
           {arriveTangent=-0.574, interpMode=RCIM_Cubic, leaveTangent=-0.574, time=0.0295, value=0.4}
           {interpMode=RCIM_Cubic, leaveTangent=-1.3593, time=0.5226, value=0.3}
          ]}
         GreenCurve {keys [
           <at 0 0.6>
           {arriveTangent=-0.574, interpMode=RCIM_Cubic, leaveTangent=-0.574, time=0.0295, value=0.4}
           {interpMode=RCIM_Cubic, leaveTangent=-2.7288, time=0.5226, value=0.3}
          ]}
         RedCurve {keys [<at 0 3>, {arriveTangent=-3.8268, interpMode=RCIM_Cubic, leaveTangent=-3.8268, time=0.0295, value=2}, <at 0.5226 1 RCIM_Cubic>]}
        }})
      2=?
      3=?
      Index=$System.Variant
      Num=::NewEnumerator0
     ))
    ScaleA=true
    ScaleRGB=true
   )
   &ScaleSpriteSize(
    Non-Uniform Curve Sprite Scale=DataInterface() {propertyValues {
      XCurve {keys [
        {interpMode=RCIM_Cubic, tangentMode=RCTM_User, time=0, value=0.2}
        {arriveTangent=3.4491, interpMode=RCIM_Cubic, leaveTangent=3.4491, tangentMode=RCTM_User, time=0.08, value=0.3}
        {arriveTangent=-2.6795, interpMode=RCIM_Cubic, leaveTangent=-2.6795, tangentMode=RCTM_User, time=0.2, value=1.2}
        {arriveTangent=-2.2231, interpMode=RCIM_Cubic, leaveTangent=-2.2231, tangentMode=RCTM_User, time=0.3, value=0.6}
        {interpMode=RCIM_Cubic, tangentMode=RCTM_User, time=1, value=0}
       ]}
      YCurve {keys [
        {interpMode=RCIM_Cubic, tangentMode=RCTM_User, time=0, value=0.2}
        {interpMode=RCIM_Cubic, tangentMode=RCTM_User, time=0.2, value=1.025}
        {tangentMode=RCTM_User, time=1, value=0.8}
       ]}
     }}
    Scale Sprite Size Mode=::NewEnumerator3
   ) {module=ScaleSpriteSize001}
   /SetVariables(
    Particles.ColorDark:LinearColor=&MergeHelper_ColorSwitch(
     0=&ColorFromCurve(ColorCurve=DataInterface() {propertyValues {
        AlphaCurve {keys [<at 0 1 RCIM_Cubic>,<at 0.519 0 RCIM_Cubic>]}
        BlueCurve {keys [<at 0 0.2 RCIM_Cubic>]}
        GreenCurve {keys [<at 0 0.0933 RCIM_Cubic>]}
        RedCurve {keys [<at 0 0.04 RCIM_Cubic>]}
       }})
     1=&ColorFromCurve(ColorCurve=DataInterface() {propertyValues {
        AlphaCurve {keys [<at 0 1 RCIM_Cubic>,<at 0.519 0 RCIM_Cubic>]}
        BlueCurve {keys [<at 0 0.04 RCIM_Cubic>]}
        GreenCurve {keys [<at 0 0.04 RCIM_Cubic>]}
        RedCurve {keys [<at 0 0.2 RCIM_Cubic>]}
       }})
     2=?
     3=?
     Index=$System.Variant
     Num=::NewEnumerator0
    )
    Particles.SpriteSizeDark:Vector2f=&Multiply_Vector2D(A=$Particles.SpriteSize, B=<1 1>)
   ) {isSetParameters=true}
  ]
  rendererClasses [&NiagaraSpriteRendererProperties]
  renderers [
   {rendererClass=&NiagaraSpriteRendererProperties}
   {rendererClass=&NiagaraSpriteRendererProperties, rendererIndex=1}
  ]
 }
 DirectionalStreaks {
  emitterUpdateScript [
   &EmitterState()
   &SpawnBurst_Instantaneous(
    Spawn Count=3
    Spawn Probability=&FN_FrontEndFloatModifier(Gameplay=1, Front End=0)
    Use Spawn Probability=true
    UseLoopCountLimit=false
   )
  ]
  particleSpawnScript [
   &InitializeParticle(
    Color=&MergeHelper_ColorSwitch(
     0=&RandomRangeLinearColor(Minimum=<rgba 0.2 0.4667 1 1>, Maximum=<rgba 0.2 0.3333 1 1>)
     1=<rgba 1 0.02 0.02 1>
     2=?
     3=?
     Index=$System.Variant
     Num=::NewEnumerator0
    )
    Lifetime Max=0.2
    Lifetime Min=0.14
    Lifetime Mode=::NewEnumerator1
    Sprite Size Max=<25 80>
    Sprite Size Min=<20 60>
    Sprite Size Mode=::NewEnumerator2
    Sprite UV Mode=::NewEnumerator1
    UsePositionOffset=false
    Write Material Random=true
   ) {module=InitializeParticle001}
   &AddVelocity(Module.Use Falloff Distance=false, Override Seed=false, Velocity Mode=::NewEnumerator1) {module=AddVelocity001}
   &SubUVAnimation(End Frame=3, SubUV Animation Mode:ENiagara_SubUVLookupMode=::NewEnumerator2)
   &AddVelocity(
    Velocity=$System.DirectionVector
    Velocity Speed Scale=&MakeCustomFloatFromBool(Bool=$User.bIsFrontEndPreview, True Float=0, False Float=&RandomRangeFloat(Minimum=60, Maximum=100))
   )
   &CameraOffset(Camera Offset Amount=25)
  ]
  particleUpdateScript [
   &ParticleState()
   &ScaleSpriteSize(Uniform Curve Sprite Scale=DataInterface() {propertyValues {Curve {keys [{tangentMode=RCTM_User, time=0, value=0}, {tangentMode=RCTM_User, time=0.2, value=1}]}}}) {module=ScaleSpriteSize001}
   &SolveForcesAndVelocity(Clamp Velocity=false, Limit Acceleration=false)
   &ScaleSpriteSize(
    Scale Factor=&Vector2DFromCurve(
     Vector2Curve=DataInterface() {propertyValues {
       XCurve {keys [{arriveTangent=-2.3004, interpMode=RCIM_Cubic, leaveTangent=-2.3004, tangentMode=RCTM_User, time=0, value=1}, <at 1 0.1 RCIM_Cubic>]}
       YCurve {keys [{arriveTangent=1.8515, interpMode=RCIM_Cubic, leaveTangent=1.8515, tangentMode=RCTM_User, time=0, value=0.3}, <at 1 1.5 RCIM_Cubic>]}
      }}
     ScaleCurve=&Multiply_Vector2D_ByFloat(
      Float=&MakeCustomFloatFromBool(Bool=$User.bIsFrontEndPreview, True Float=0.6, False Float=1)
      Vector2D=<1.5 1>
     )
    )
    Scale Sprite Size Mode=::NewEnumerator2
   )
   &ScaleColor(
    Scale Alpha=&FloatFromCurve(
     FloatCurve=DataInterface() {propertyValues {Curve {keys [<at 0 0 RCIM_Cubic>,<at 0.1 1 RCIM_Cubic>,<at 1 0>]}}}
     Scale Curve=&FNTimeOfDayScalar(Night Scale=0.5, Morning Scale=1, Evening Scale=0.5)
    )
    Scale RGB=&VectorFromFloat(Value=&FloatFromCurve(
      FloatCurve=DataInterface() {propertyValues {Curve {keys [
          {arriveTangent=-52.7997, interpMode=RCIM_Cubic, leaveTangent=-52.7997, tangentMode=RCTM_User, time=0, value=10}
          {arriveTangent=-3.0339, interpMode=RCIM_Cubic, leaveTangent=-3.0339, tangentMode=RCTM_User, time=0.2, value=1.5}
          <at 1 1 RCIM_Cubic>
         ]}}}
      Scale Curve=2
     ))
    ScaleA=true
    ScaleRGB=true
   )
  ]
  rendererClasses [&NiagaraSpriteRendererProperties]
  renderers [{rendererClass=&NiagaraSpriteRendererProperties}]
 }
 Flash_Elements {
  emitterUpdateScript [
   &EmitterState()
   &SpawnBurst_Instantaneous(Spawn Group=1, Use Spawn Probability=false, UseLoopCountLimit=false) {module=SpawnBurst_Instantaneous001}
   &SpawnBurst_Instantaneous(
    Spawn Count=&RandomRangeInt(Minimum=6, Maximum=8)
    Use Spawn Probability=false
    UseLoopCountLimit=false
   )
  ]
  particleSpawnScript [
   &InitializeParticle(Color Mode=::NewEnumerator0, Lifetime=0.23, UsePositionOffset=false, Write Material Random=true)
   &ShapeLocation(MaskSpawnGroup=false, Override Seed=false, Sphere Radius=1, UseOwnerScale=true)
   &SpriteFacingAndAlignment(Sprite Alignment=$Particles.ShapeLocation.ShapeNormal, WriteAlignment=true, WriteFacing=false)
   &FN_PerParticleLOD(Enable SpriteSize Distance Scaling=true, Sprite Size Scale Far=<2.5 2.5>)
   /SetVariables(Particles.SpriteSize=<1 1>, Particles.VisibilityTag=&SpawnGroup()) {isSetParameters=true}
   /SetVariables(
    Particles.Flash Spike Color:LinearColor=<rgba 4.2448 5 4.9846 2>
    Particles.Flash Spike Lifetime:NiagaraFloat=&RandomRangeFloat(Minimum=0.05, Maximum=0.07)
    Particles.Flash Spike Sprite Size:Vector2f=&Multiply_Vector2D_ByFloat(
     Float=&FN_FrontEndFloatModifier(Gameplay=0.8, Front End=$User.FrontEndScale)
     Vector2D=&RandomRangeVector2D(Minimum=<5 40>, Maximum=<10 60>)
    )
   ) {isSetParameters=true}
   /SetVariables(
    Particles.Flash Star Cam Offset=15
    Particles.Flash Star Lifetime=0.15
    Particles.Flash Star Rotation:NiagaraFloat=&RandomRangeFloat(Minimum=0, Maximum=360)
    Particles.Flash Star Sprite Size:Vector2f=&Multiply_Vector2D_ByFloat(
     Float=&FN_FrontEndFloatModifier(Gameplay=1, Front End=$User.FrontEndScale)
     Vector2D=$Particles.SpriteSize
    )
    Particles.Flash Star Sprite Size Alt:Vector2f=&Multiply_Vector2D_ByFloat(
     Float=&Multiply_Float(
      A=0.85
      B=&Multiply_Float(
       A=&FN_FrontEndFloatModifier(Gameplay=1, Front End=$User.FrontEndScale)
       B=&RandomRangeFloat(Minimum=120, Maximum=150)
      )
     )
     Vector2D=$Particles.SpriteSize
    )
    Particles.Flash Star SubUV Index:NiagaraFloat=&RandomRangeFloat(Minimum=0, Maximum=3)
    Particles.Flash Star SubUV Index Alt:NiagaraFloat=&RandomRangeFloat(Minimum=0, Maximum=3)
   ) {isSetParameters=true}
   /SetVariables(
    Particles.Flash Flare Lifetime=0.2
    Particles.Flash Flare Sprite Size:Vector2f=&Vector2DFromFloat(Value=&Multiply_Float(A=100, B=&FN_FrontEndFloatModifier(Gameplay=1, Front End=$User.FrontEndScale)))
   ) {isSetParameters=true}
   &DynamicMaterialParameters(
    Index 0 Param 1=0.2
    Index 0 Param 2=1
    Module.Index0Float0Write=true
    Module.Index0Float1Write=true
    Module.Index0Float2Write=true
    Module.Index0Float3Write=true
   )
   &CameraOffset(Camera Offset Amount=&FN_FrontEndFloatModifier(Gameplay=20, Front End=10))
  ]
  particleUpdateScript [
   &ParticleState()
   /SetVariables(Particles.Flash Flare Sprite Size:Vector2f=&Multiply_Vector2D_ByFloat(
     Float=&FloatFromCurve(FloatCurve=DataInterface() {propertyValues {Curve {keys [<at 0 1.5 RCIM_Cubic>, {interpMode=RCIM_Cubic, leaveTangent=-3.4235, time=1, value=1}]}}})
     Vector2D=$Particles.Initial.Flash Flare Sprite Size
    )) {isSetParameters=true}
   /SetVariables(
    Particles.Update Flash Flare Age:NiagaraFloat=&Divide_Float(A=$Particles.Age, B=$Particles.Initial.Flash Flare Lifetime)
    Particles.Update Flash Spike Age:NiagaraFloat=&Divide_Float(A=$Particles.Age, B=$Particles.Initial.Flash Spike Lifetime)
    Particles.Update Flash Star Age:NiagaraFloat=&Divide_Float(A=$Particles.Age, B=$Particles.Initial.Flash Star Lifetime)
   ) {isSetParameters=true}
   /SetVariables(
    Particles.Update Flash Spike Color:LinearColor=&MakeLinearColorFromVectorAndFloat(
     "Vector (RGB)"=&MakeVectorFromLinearColorRGB(Color=&MultiplyLinearColor(
       Linear Color=$Particles.Initial.Flash Spike Color
       Scale Factor=&MakeVector4_FromLinearColor(LinearColor=&MergeHelper_ColorSwitch(
         0=&ColorFromCurve(ColorCurve=DataInterface() {propertyValues {
            AlphaCurve {keys [<at 0 1 RCIM_Cubic>]}
            BlueCurve {keys [<at 0 1 RCIM_Cubic>,<at 1 1 RCIM_Cubic>]}
            GreenCurve {keys [<at 0 0.8934 RCIM_Cubic>,<at 1 0.2835 RCIM_Cubic>]}
            RedCurve {keys [<at 0 0.6803 RCIM_Cubic>,<at 1 0 RCIM_Cubic>]}
           }})
         1=&ColorFromCurve(ColorCurve=DataInterface() {propertyValues {
            AlphaCurve {keys [<at 0 1 RCIM_Cubic>]}
            BlueCurve {keys [<at 0 0.4 RCIM_Cubic>,<at 1 0 RCIM_Cubic>]}
            GreenCurve {keys [<at 0 0.4 RCIM_Cubic>,<at 1 0 RCIM_Cubic>]}
            RedCurve {keys [<at 0 1 RCIM_Cubic>,<at 1 1 RCIM_Cubic>]}
           }})
         2=?
         3=?
         Index=$System.Variant
         Num=::NewEnumerator0
        ))
      ))
     "Float (Alpha)"=&FloatFromCurve(FloatCurve=DataInterface() {propertyValues {Curve {keys [<at 0.5 1>,<at 1 0>]}}}, CurveIndex=$Particles.Update Flash Spike Age)
    )
    Particles.Update Flash Spike Sprite Size:Vector2f=&Multiply_Vector2D_ByFloat(
     Float=&FloatFromCurve(
      FloatCurve=DataInterface() {propertyValues {Curve {keys [<at 0 0.6>,<at 0.9 2>]}}}
      CurveIndex=$Particles.Update Flash Spike Age
      Scale Curve=&FN_FrontEndFloatModifier(Gameplay=1, Front End=$System.FrontendScale)
     )
     Vector2D=$Particles.Initial.Flash Spike Sprite Size
    )
    Particles.Update Flash Star Color:LinearColor=&ColorFromCurve(
     ColorCurve=DataInterface() {propertyValues {
       AlphaCurve {keys [<at 0 1.5>,<at 0.8992 0>]}
       BlueCurve {keys [<at 0 10>,<at 0.3455 2>,<at 0.7329 1.5>]}
       GreenCurve {keys [<at 0 8.6711>,<at 0.3455 1.9946>,<at 0.7329 0.4496>]}
       RedCurve {keys [<at 0 1.7923>,<at 0.3455 0>,<at 0.7329 0.1745>]}
      }}
     CurveIndex=$Particles.Update Flash Star Age
    )
    Particles.Update Flash Star Sprize Size:Vector2f=&Multiply_Vector2D_ByFloat(
     Float=&FloatFromCurve(
      FloatCurve=DataInterface() {propertyValues {Curve {keys [
          {arriveTangent=2.7013, interpMode=RCIM_Cubic, leaveTangent=2.7013, tangentMode=RCTM_User, time=0, value=0.15}
          {interpMode=RCIM_Cubic, tangentMode=RCTM_User, time=1, value=1}
         ]}}}
      CurveIndex=$Particles.Update Flash Star Age
     )
     Vector2D=$Particles.Initial.Flash Star Sprite Size
    )
    Particles.Update Flash Star Sprize Size Alt:Vector2f=&Multiply_Vector2D_ByFloat(
     Float=&FloatFromCurve(
      FloatCurve=DataInterface() {propertyValues {Curve {keys [
          {arriveTangent=2.7013, interpMode=RCIM_Cubic, leaveTangent=2.7013, tangentMode=RCTM_User, time=0, value=0}
          {interpMode=RCIM_Cubic, tangentMode=RCTM_User, time=1, value=1}
         ]}}}
      CurveIndex=$Particles.Update Flash Star Age
     )
     Vector2D=$Particles.Initial.Flash Star Sprite Size Alt
    )
   ) {isSetParameters=true}
   /SetVariables(Particles.Update Flash Flare Color:LinearColor=&MakeLinearColorFromVectorAndFloat(
     "Vector (RGB)"=&Multiply_Vector(
      A=&MakeVectorFromLinearColorRGB(Color=&MultiplyLinearColor(
        Linear Color=$Particles.Initial.Flash Flare Color
        Scale Factor=&MakeVector4_FromLinearColor(LinearColor=&MergeHelper_ColorSwitch(
          0=&ColorFromCurve(ColorCurve=DataInterface() {propertyValues {
             AlphaCurve {keys [<at 0 1 RCIM_Cubic>]}
             BlueCurve {keys [<at 0 3>,<at 1 1 RCIM_Cubic>]}
             GreenCurve {keys [<at 0 1.4>,<at 1 0.3333 RCIM_Cubic>]}
             RedCurve {keys [<at 0 0.6>,<at 1 0.2 RCIM_Cubic>]}
            }})
          1=&ColorFromCurve(ColorCurve=DataInterface() {propertyValues {
             AlphaCurve {keys [<at 0 1 RCIM_Cubic>]}
             BlueCurve {keys [<at 0 0.3>,<at 1 0.1 RCIM_Cubic>]}
             GreenCurve {keys [<at 0 0.3>,<at 1 0.1 RCIM_Cubic>]}
             RedCurve {keys [<at 0 3>,<at 1 1 RCIM_Cubic>]}
            }})
          2=?
          3=?
          Index=$System.Variant
          Num=::NewEnumerator0
         ))
       ))
      B=&VectorFromFloat(Value=&Multiply_Float(A=1.5, B=&FN_FrontEndFloatModifier(Gameplay=1, Front End=$System.FrontendScale)))
     )
     "Float (Alpha)"=&FloatFromCurve(
      FloatCurve=DataInterface() {propertyValues {Curve {keys [
          {arriveTangent=-4.121, interpMode=RCIM_Cubic, leaveTangent=-4.121, tangentMode=RCTM_User, time=0, value=1}
          {interpMode=RCIM_Cubic, tangentMode=RCTM_User, time=0.7, value=0}
         ]}}}
      Scale Curve=&Multiply_Float(A=8, B=&FN_Select_Float_from_Time_Of_Day(Morning=0.8, Day=1, Evening=0.8, Night=0.5))
     )
    )) {isSetParameters=true}
   &FN_Light_Attributes(
    Color=&MergeHelper_ColorSwitch(
     0=&ColorFromCurve(ColorCurve=DataInterface() {propertyValues {
        AlphaCurve {keys [<at 0 1 RCIM_Cubic>,<at 1 0>]}
        BlueCurve {keys [<at 0 3 RCIM_Cubic>,<at 1 1 RCIM_Cubic>]}
        GreenCurve {keys [<at 0 1.2 RCIM_Cubic>,<at 1 0.3333 RCIM_Cubic>]}
        RedCurve {keys [<at 0 0.3 RCIM_Cubic>,<at 1 0.2 RCIM_Cubic>]}
       }})
     1=&ColorFromCurve(ColorCurve=DataInterface() {propertyValues {
        AlphaCurve {keys [<at 0 1 RCIM_Cubic>,<at 1 0>]}
        BlueCurve {keys [<at 0 0.15 RCIM_Cubic>,<at 1 0.1 RCIM_Cubic>]}
        GreenCurve {keys [<at 0 0.15 RCIM_Cubic>,<at 1 0.1 RCIM_Cubic>]}
        RedCurve {keys [<at 0 3 RCIM_Cubic>,<at 1 1 RCIM_Cubic>]}
       }})
     2=?
     3=?
     Index=$System.Variant
     Num=::NewEnumerator0
    )
    Create Color Attribute=true
    Create Exponent Attribute=true
    Create Position Attribute=true
    Create Radius Attribute=true
    Exponent=5
    Position=<15 0 0>
    Radius=100
    Scale Color=&FloatFromCurve(
     FloatCurve=DataInterface() {propertyValues {Curve {keys [
         {arriveTangent=-1.8529, interpMode=RCIM_Cubic, leaveTangent=-1.8529, tangentMode=RCTM_User, time=0.2, value=1}
         {arriveTangent=-0.5426, interpMode=RCIM_Cubic, leaveTangent=-0.5426, tangentMode=RCTM_User, time=0.8, value=0}
        ]}}}
     Scale Curve=&Multiply_Float(A=18, B=&FN_Select_Float_from_Time_Of_Day(Morning=0.75, Day=1, Evening=0.75, Night=0.5))
    )
   )
   /SetVariables(Particles.ColorDark:LinearColor=&MergeHelper_ColorSwitch(
     0=&ColorFromCurve(ColorCurve=DataInterface() {propertyValues {
        AlphaCurve {keys [<at 0.2394 1 RCIM_Cubic>,<at 0.9406 0 RCIM_Cubic>]}
        BlueCurve {keys [<at 0 0.0417 RCIM_Cubic>]}
        GreenCurve {keys [<at 0 0.0203 RCIM_Cubic>]}
        RedCurve {keys [<at 0 0.005 RCIM_Cubic>]}
       }})
     1=&ColorFromCurve(ColorCurve=DataInterface() {propertyValues {
        AlphaCurve {keys [<at 0.2394 1 RCIM_Cubic>,<at 0.9406 0 RCIM_Cubic>]}
        BlueCurve {keys [<at 0 0.005 RCIM_Cubic>]}
        GreenCurve {keys [<at 0 0.005 RCIM_Cubic>]}
        RedCurve {keys [<at 0 0.0417 RCIM_Cubic>]}
       }})
     2=?
     3=?
     Index=$System.Variant
     Num=::NewEnumerator0
    )) {isSetParameters=true}
   /SetVariables(Particles.DarkGlowSpriteSize:Vector2f=&Multiply_Vector2D_ByFloat(
     Float=&FloatFromCurve(
      FloatCurve=DataInterface() {propertyValues {Curve {keys [
          <at 0 0 RCIM_Cubic>
          {arriveTangent=-2.6948, interpMode=RCIM_Cubic, leaveTangent=-2.6948, tangentMode=RCTM_User, time=0.1, value=1}
          {interpMode=RCIM_Cubic, tangentMode=RCTM_User, time=0.45, value=0.4}
         ]}}}
      Scale Curve=&Multiply_Float(A=0, B=&FN_FrontEndFloatModifier(Gameplay=1, Front End=$User.FrontEndScale))
     )
     Vector2D=$Particles.Initial.Flash Flare Sprite Size
    )) {isSetParameters=true}
  ]
  rendererClasses [&NiagaraSpriteRendererProperties, &NiagaraLightRendererProperties]
  renderers [
   {rendererClass=&NiagaraSpriteRendererProperties}
   {rendererClass=&NiagaraSpriteRendererProperties, rendererIndex=1}
   {rendererClass=&NiagaraSpriteRendererProperties, rendererIndex=2}
   {rendererClass=&NiagaraSpriteRendererProperties, rendererIndex=3}
   {rendererClass=&NiagaraLightRendererProperties, rendererIndex=4}
   {rendererClass=&NiagaraSpriteRendererProperties, rendererIndex=5}
  ]
 }
 Impact_Smoke {
  emitterUpdateScript [
   &EmitterState()
   &SpawnBurst_Instantaneous(
    Spawn Count=&RandomRangeInt(Minimum=2, Maximum=3)
    Use Spawn Probability=false
    UseLoopCountLimit=false
   )
  ]
  particleSpawnScript [
   &InitializeParticle(
    Color Maximum=<rgba 0.76 1 1 0.5>
    Color Minimum=<rgba 0.8 0.917 1 0.2>
    Color Mode=::NewEnumerator2
    Lifetime Max=0.4
    Lifetime Min=0.3
    Lifetime Mode=::NewEnumerator1
    Sprite Size Mode=::NewEnumerator3
    Uniform Sprite Size=&Multiply_Float(
     A=&FN_FrontEndFloatModifier(Gameplay=1, Front End=$System.FrontendScale)
     B=&RandomRangeFloat(Minimum=100, Maximum=130)
    )
    UsePositionOffset=false
    Write Material Random=true
   )
   &CameraOffset(Camera Offset Amount=&RandomRangeFloat(Minimum=40, Maximum=50, Evaluation Type=::NewEnumerator1))
   &ShapeLocation(
    Hemisphere Distribution=<1 0.5>
    MaskSpawnGroup=false
    Override Seed=false
    Rotation Angle=90
    Rotation Mode=::NewEnumerator1
    Sphere Radius=&FN_FrontEndFloatModifier(Gameplay=10, Front End=5)
    UseOwnerScale=true
   )
   &AddVelocity(Velocity=&Multiply_VectorByFloat(Vector=$Particles.ShapeLocation.ShapeVector, Float=10))
  ]
  particleUpdateScript [
   &ParticleState()
   &ScaleSpriteSize(Uniform Curve Sprite Scale=DataInterface() {propertyValues {Curve {keys [
        {arriveTangent=1.7761, interpMode=RCIM_Cubic, leaveTangent=1.7761, tangentMode=RCTM_User, time=0, value=0}
        {interpMode=RCIM_Cubic, tangentMode=RCTM_User, time=1, value=1}
       ]}}})
   &ScaleColor(
    Scale Alpha=&FloatFromCurve(FloatCurve=DataInterface() {propertyValues {Curve {keys [
         {arriveTangent=2.4358, interpMode=RCIM_Cubic, leaveTangent=2.4358, tangentMode=RCTM_User, time=0, value=0}
         <at 0.25 0.5 RCIM_Cubic>
         <at 1 0 RCIM_Cubic>
        ]}}})
    Scale RGB=&MakeVectorFromLinearColorRGB(Color=&MergeHelper_ColorSwitch(
      0=&ColorFromCurve(ColorCurve=DataInterface() {propertyValues {
         AlphaCurve {keys [<at 0 1 RCIM_Cubic>]}
         BlueCurve {keys [<at 0.2265 1 RCIM_Cubic>,<at 0.7544 0.8 RCIM_Cubic>]}
         GreenCurve {keys [<at 0.2265 0.6795 RCIM_Cubic>,<at 0.7544 0.5867 RCIM_Cubic>]}
         RedCurve {keys [<at 0.2265 0.4505 RCIM_Cubic>,<at 0.7544 0.48 RCIM_Cubic>]}
        }})
      1=&ColorFromCurve(ColorCurve=DataInterface() {propertyValues {
         AlphaCurve {keys [<at 0 1 RCIM_Cubic>]}
         BlueCurve {keys [<at 0.2265 0.4505 RCIM_Cubic>,<at 0.7544 0.48 RCIM_Cubic>]}
         GreenCurve {keys [<at 0.2265 0.4505 RCIM_Cubic>,<at 0.7544 0.48 RCIM_Cubic>]}
         RedCurve {keys [<at 0.2265 1 RCIM_Cubic>,<at 0.7544 0.8 RCIM_Cubic>]}
        }})
      2=?
      3=?
      Index=$System.Variant
      Num=::NewEnumerator0
     ))
    ScaleA=true
    ScaleRGB=true
   )
   &DynamicMaterialParameters(
    Index 0 Param 1=&Add_Float(
     A=&RandomRangeFloat(Minimum=0, Maximum=1)
     B=&FloatFromCurve(
      FloatCurve=DataInterface() {propertyValues {Curve {keys [
          <at 0 0 RCIM_Cubic>
          {arriveTangent=1.5947, interpMode=RCIM_Cubic, leaveTangent=1.5947, tangentMode=RCTM_User, time=0.2557, value=0.5257}
          <at 1 1 RCIM_Cubic>
         ]}}}
      CurveIndex=$Particles.Age
      Scale Curve=0.25
     )
    )
    Module.Param0WriteEnabled=true
    Module.Param1WriteEnabled=true
    Module.Param2WriteEnabled=true
    Module.Param3WriteEnabled=true
   ) {module=DynamicMaterialParameters001}
   &SolveForcesAndVelocity(Clamp Velocity=false, Limit Acceleration=false)
  ]
  rendererClasses [&NiagaraSpriteRendererProperties]
  renderers [
   {rendererClass=&NiagaraSpriteRendererProperties}
   {rendererClass=&NiagaraSpriteRendererProperties, rendererIndex=1}
  ]
 }
 LongFlare {
  emitterUpdateScript [
   &EmitterState(Life Cycle Mode=::NewEnumerator1, Loop Behavior=::NewEnumerator1, Loop Delay=0.1, Loop Duration=1, UseLoopDelay=false)
   &SpawnBurst_Instantaneous(Spawn Count=1, Use Spawn Probability=false)
  ]
  particleSpawnScript [
   &InitializeParticle(
    Color=&MergeHelper_ColorSwitch(0=<rgba 1.5 2.0833 5 0.5>, 1=<rgba 5 0.1 0.1 0.5>, 2=?, 3=?, Index=$System.Variant, Num=::NewEnumerator0)
    Lifetime=0.27
    Sprite Rotation Mode=::NewEnumerator2
    Sprite Size Mode=::NewEnumerator3
    Uniform Sprite Size=200
    UsePositionOffset=false
    Write Material Random=true
   )
   &DynamicMaterialParameters(
    Index 0 Param 1=&RandomRangeFloat(Minimum=0.4, Maximum=0.6)
    Index 0 Param 2=1
    Index 0 Param 3=1
    Index 0 Param 4=1
    Module.Param0WriteEnabled=true
    Module.Param1WriteEnabled=true
    Module.Param2WriteEnabled=true
    Module.Param3WriteEnabled=true
   )
   /SetVariables(Particles.Dark_Glow_CameraOffset=&MakeCustomFloatFromBool(Bool=$User.bIsFrontEndPreview, True Float=1, False Float=50)) {isSetParameters=true}
   &CameraOffset(Camera Offset Amount=&MakeCustomFloatFromBool(Bool=$User.bIsFrontEndPreview, True Float=1, False Float=30))
  ]
  particleUpdateScript [
   /SetVariables(Particles.Dark_Glow=&MakeLinearColorFromVectorAndFloat(
     "Vector (RGB)"=&MakeVectorFromLinearColorRGB(Color=&MultiplyLinearColor(
       Linear Color=&MergeHelper_ColorSwitch(0=<rgba 0.05 0.0583 0.1 1>, 1=<rgba 0.15 0.015 0.015 1>, 2=?, 3=?, Index=$System.Variant, Num=::NewEnumerator0)
       Scale Factor=&MakeVector4_FromFloat(Float=1)
      ))
     "Float (Alpha)"=&Multiply_Float(
      A=&Multiply_Float(A=&FNTimeOfDayScalar(Night Scale=0.3, Morning Scale=0.5, Evening Scale=0.5), B=1.2)
      B=&FloatFromCurve(
       FloatCurve=DataInterface() {propertyValues {Curve {keys [
           {interpMode=RCIM_Cubic, tangentMode=RCTM_User, time=0, value=0}
           {interpMode=RCIM_Cubic, tangentMode=RCTM_User, time=0.2, value=1}
           {interpMode=RCIM_Cubic, tangentMode=RCTM_User, time=0.7, value=0.6}
           {interpMode=RCIM_Cubic, tangentMode=RCTM_User, time=1, value=0}
          ]}}}
       Scale Curve=&FN_FrontEndFloatModifier(Gameplay=1, Front End=$User.FrontEndScale)
      )
     )
    )) {isSetParameters=true}
   /SetVariables(Particles.Dark_Glow_Size=&Multiply_Vector2(
     A=&Vector2DFromFloat(Value=220)
     B=&Vector2DFromFloat(Value=&FloatFromCurve(
       FloatCurve=DataInterface() {propertyValues {Curve {keys [<at 0 1.5 RCIM_Cubic>,<at 1 0.5>]}}}
       Scale Curve=&FN_FrontEndFloatModifier(Gameplay=1, Front End=0.4)
      ))
    )) {isSetParameters=true}
   &ParticleState()
   &ScaleColor(
    Scale Alpha=&Multiply_Float(
     A=&FNTimeOfDayScalar(Night Scale=0.5, Morning Scale=0.5, Evening Scale=0.5)
     B=&FloatFromCurve(
      FloatCurve=DataInterface() {propertyValues {Curve {keys [<at 0 0>,<at 0.2 1>,<at 0.8 1>,<at 1 0>]}}}
      Scale Curve=&FNTimeOfDayScalar(Night Scale=0.5, Morning Scale=0.5, Evening Scale=0.5)
     )
    )
    Scale RGB=<2 2 2>
    ScaleA=true
    ScaleRGB=true
   )
   &ScaleSpriteSize(Scale Factor=&Vector2DFromFloat(Value=&FloatFromCurve(
      FloatCurve=DataInterface() {propertyValues {Curve {keys [<at 0 1.5 RCIM_Cubic>,<at 1 0>]}}}
      Scale Curve=&FN_FrontEndFloatModifier(Gameplay=1, Front End=0.4)
     )))
  ]
  rendererClasses [&NiagaraSpriteRendererProperties]
  renderers [
   {rendererClass=&NiagaraSpriteRendererProperties}
   {rendererClass=&NiagaraSpriteRendererProperties, rendererIndex=1}
  ]
 }
 Shockwave {
  bEnabled=false
  emitterUpdateScript [
   &EmitterState()
   &SpawnBurst_Instantaneous(Spawn Count=1, Use Spawn Probability=false)
  ]
  particleSpawnScript [
   &InitializeParticle(
    Color=<rgba 1 1 1 5>
    Lifetime=&FN_FrontEndFloatModifier(Gameplay=0.33, Front End=0.27)
    Mesh Scale Mode=::NewEnumerator1
    Mesh Uniform Scale Max=1
    Mesh Uniform Scale Min=0.9
    UsePositionOffset=false
    Write Material Random=true
   ) {module=InitializeParticle001}
   &InitialMeshOrientation(
    Mesh Axis To Orient=<0 0 1>
    Mesh Orientation Mode=::NewEnumerator2
    Rotation=&MakeVector(X=0, Y=0, Z=&RandomRangeFloat(Minimum=0, Maximum=1))
    Use Rotation Vector=true
   )
  ]
  particleUpdateScript [
   &ParticleState()
   &DynamicMaterialParameters(
    Index 0 Param 1=&FloatFromCurve(FloatCurve=DataInterface() {propertyValues {Curve {keys [{arriveTangent=-2.1015, interpMode=RCIM_Cubic, leaveTangent=-2.1015, tangentMode=RCTM_User, time=0, value=0.5}, <at 1 -0.15>]}}})
    Index 0 Param 2=&FloatFromCurve(FloatCurve=DataInterface() {propertyValues {Curve {keys [{arriveTangent=-0.3464, interpMode=RCIM_Cubic, leaveTangent=-0.3464, tangentMode=RCTM_User, time=0, value=1}, <at 1 0.9>]}}})
    Index 0 Param 3=&FloatFromCurve(FloatCurve=DataInterface() {propertyValues {Curve {keys [
         {arriveTangent=8.3298, interpMode=RCIM_Cubic, leaveTangent=8.3298, tangentMode=RCTM_User, time=0, value=0}
         {arriveTangent=0.3366, interpMode=RCIM_Cubic, leaveTangent=0.3366, tangentMode=RCTM_User, time=0.15, value=1}
         <at 1 1.2>
        ]}}})
    Module.Param0WriteEnabled=true
    Module.Param1WriteEnabled=true
    Module.Param2WriteEnabled=true
    Module.Param3WriteEnabled=false
   )
   &ScaleColor(
    Scale Mode=::NewEnumerator1
    Scale RGBA=&MakeVector4_FromLinearColor(LinearColor=&MergeHelper_ColorSwitch(
      0=&ColorFromCurve(ColorCurve=DataInterface() {propertyValues {
         AlphaCurve {keys [<at 0 1.5>,<at 1 0>]}
         BlueCurve {keys [<at 0 5>,<at 0.6824 2>]}
         GreenCurve {keys [<at 0 2.3333>,<at 0.6824 0.6667>]}
         RedCurve {keys [<at 0 1>,<at 0.6824 0.4>]}
        }})
      1=&ColorFromCurve(ColorCurve=DataInterface() {propertyValues {
         AlphaCurve {keys [<at 0 1.5>,<at 1 0>]}
         BlueCurve {keys [<at 0 0>,<at 0.6824 0>]}
         GreenCurve {keys [<at 0 0>,<at 0.6824 0>]}
         RedCurve {keys [<at 0 10>,<at 0.6824 2>]}
        }})
      2=?
      3=?
      Index=$System.Variant
      Num=::NewEnumerator0
     ))
    ScaleRGBA=true
   )
   &ScaleMeshSize(Scale Factor=&VectorFromFloat(Value=&FloatFromCurve(
      FloatCurve=DataInterface() {propertyValues {Curve {keys [
          {arriveTangent=6.1565, interpMode=RCIM_Cubic, leaveTangent=6.1565, tangentMode=RCTM_User, time=0, value=0}
          {arriveTangent=0.3937, interpMode=RCIM_Cubic, leaveTangent=0.3937, tangentMode=RCTM_User, time=0.1565, value=0.791}
          {arriveTangent=0.1097, interpMode=RCIM_Cubic, leaveTangent=0.1097, tangentMode=RCTM_User, time=1, value=1}
         ]}}}
      Scale Curve=&FN_FrontEndFloatModifier(Gameplay=1, Front End=$User.FrontEndScale)
     )))
   &CameraOffset(Camera Offset Amount=10)
  ]
  rendererClasses [&NiagaraMeshRendererProperties]
  renderers [{rendererClass=&NiagaraMeshRendererProperties, rendererIndex=0}]
 }
 Sparks {
  emitterUpdateScript [
   &EmitterState(Life Cycle Mode=::NewEnumerator1, Loop Behavior=::NewEnumerator1, Loop Delay=0.05, Loop Duration=1, UseLoopDelay=true)
   &SpawnBurst_Instantaneous(Spawn Count=12, Use Spawn Probability=false)
   &SpawnBurst_Instantaneous(Spawn Count=12, Spawn Group=1, Use Spawn Probability=false) {module=SpawnBurst_Instantaneous001}
  ]
  particleSpawnScript [
   &InitializeParticle(
    Color=&MergeHelper_ColorSwitch(0=<rgba 0.25 0.5625 1 1>, 1=<rgba 1 0.3 0.3 1>, 2=?, 3=?, Index=$System.Variant, Num=::NewEnumerator0)
    Lifetime Max=0.8
    Lifetime Min=0.3
    Lifetime Mode=::NewEnumerator1
    Sprite Size Max=<2 8>
    Sprite Size Min=<2 4>
    Sprite Size Mode=::NewEnumerator2
    UsePositionOffset=false
    Write Material Random=true
   )
   &ShapeLocation(
    Cylinder Height=&MergeHelper_FloatSwitch(0=4, 1=40, 2=?, 3=?, Index=&SpawnGroup(), Num=::NewEnumerator0)
    Cylinder Radius=2
    MaskSpawnGroup=false
    Override Seed=false
    Rotation Axis=$System.DirectionVector
    Rotation Mode=::NewEnumerator1
    Shape Primitive=::NewEnumerator1
    Surface Only=false
    UseOwnerScale=true
   )
   &AddVelocityFromPoint(
    Module.Use Falloff Distance=false
    Velocity Strength=&Multiply_Float(
     A=&Multiply_Float(
      A=&FN_FrontEndFloatModifier(Gameplay=0.7, Front End=0.35)
      B=&MergeHelper_FloatSwitch(0=1, 1=0.25, 2=?, 3=?, Index=&SpawnGroup(), Num=::NewEnumerator0)
     )
     B=&RandomRangeFloat(Minimum=500, Maximum=1500)
    )
   )
   &AddVelocity(
    Velocity=$System.DirectionVector
    Velocity Speed Scale=&Multiply_Float(
     A=&RandomRangeFloat(Minimum=600, Maximum=700)
     B=&Multiply_Float(
      A=&FN_FrontEndFloatModifier(Gameplay=1, Front End=$User.FrontEndScale)
      B=&MergeHelper_FloatSwitch(0=0, 1=1.3, 2=?, 3=?, Index=&SpawnGroup(), Num=::NewEnumerator0)
     )
    )
   )
   &CameraOffset(Camera Offset Amount=&Multiply_Float(A=30, B=$System.CameraOffsetScale))
  ]
  particleUpdateScript [
   &ParticleState()
   &ScaleColor(
    Scale RGB=&VectorFromFloat(Value=&FloatFromCurve(
      FloatCurve=DataInterface() {propertyValues {Curve {keys [{interpMode=RCIM_Cubic, tangentMode=RCTM_User, time=0, value=10}, <at 1 1>]}}}
      Scale Curve=0.3
     ))
    ScaleA=true
    ScaleRGB=true
   ) {module=ScaleColor001}
   &Drag(Drag=&RandomRangeFloat(Minimum=8, Maximum=15), Use Linear Drag=true, Use Rotational Drag=false)
   &CurlNoiseForce(Noise Strength=&RandomRangeFloat(Minimum=200, Maximum=400), Pan Noise=true)
   &GravityForce(Gravity=&RandomRangeVector(Minimum=<0 0 0>, Maximum=<0 0 -100>))
   &SolveForcesAndVelocity(Clamp Velocity=false, Limit Acceleration=false)
   &ScaleSpriteSizeBySpeed(Max Scale Factor=<1 10>, Min Scale Factor=<0.5 0.1>)
   &ScaleSpriteSize(Scale Factor=&MakeVector2D(
     X=&FloatFromCurve(FloatCurve=DataInterface() {propertyValues {Curve {keys [<at 0 0.5>,<at 0.1 1>,<at 1 0.5>]}}})
     Y=&FloatFromCurve(FloatCurve=DataInterface() {propertyValues {Curve {keys [<at 0 1>,<at 1 0>]}}})
    ))
   &ScaleSpriteSize(
    Scale Sprite Size Mode=::NewEnumerator0
    Uniform Scale Factor=&FN_FrontEndFloatModifier(Gameplay=1, Front End=$User.FrontEndScale)
   ) {module=ScaleSpriteSize001}
   &FN_Sprite_Attributes(
    Color=&MakeLinearColorFromVectorAndFloat(
     "Vector (RGB)"=&MakeVectorFromLinearColorRGB(Color=&MergeHelper_ColorSwitch(
       0=&ColorFromCurve(ColorCurve=DataInterface() {propertyValues {
          AlphaCurve {keys [<at 0 1 RCIM_Cubic>,<at 1 0 RCIM_Cubic>]}
          BlueCurve {keys [<at 0 2 RCIM_Cubic>,<at 0.4223 2 RCIM_Cubic>]}
          GreenCurve {keys [<at 0 1.3 RCIM_Cubic>,<at 0.4223 0.8333 RCIM_Cubic>]}
          RedCurve {keys [<at 0 0.6 RCIM_Cubic>,<at 0.4223 0.6 RCIM_Cubic>]}
         }})
       1=&ColorFromCurve(ColorCurve=DataInterface() {propertyValues {
          AlphaCurve {keys [<at 0 1 RCIM_Cubic>,<at 1 0 RCIM_Cubic>]}
          BlueCurve {keys [<at 0 0.1 RCIM_Cubic>,<at 0.4223 0.1 RCIM_Cubic>]}
          GreenCurve {keys [<at 0 0.1 RCIM_Cubic>,<at 0.4223 0.1 RCIM_Cubic>]}
          RedCurve {keys [<at 0 2 RCIM_Cubic>,<at 0.4223 2 RCIM_Cubic>]}
         }})
       2=?
       3=?
       Index=$System.Variant
       Num=::NewEnumerator0
      ))
     "Float (Alpha)"=&FloatFromCurve(FloatCurve=DataInterface() {propertyValues {Curve {keys [<at 0 1>,<at 0.8 0>]}}}, Scale Curve=0.25)
    )
    Create Color Attribute=true
    Create SpriteSize Attribute=true
    SpriteSize=&Multiply_Vector2D_ByFloat(
     Float=&FN_FrontEndFloatModifier(Gameplay=2, Front End=1)
     Vector2D=&Multiply_Vector2D(
      A=<5 15>
      B=&MakeVector2D(
       X=&FloatFromCurve(FloatCurve=DataInterface() {propertyValues {Curve {keys [<at 0 0.5>,<at 0.1 1>,<at 1 0.5>]}}})
       Y=&FloatFromCurve(FloatCurve=DataInterface() {propertyValues {Curve {keys [<at 0 1>,<at 1 0>]}}})
      )
     )
    )
    Write to Module Namespace=true
   )
  ]
  rendererClasses [&NiagaraSpriteRendererProperties]
  renderers [
   {rendererClass=&NiagaraSpriteRendererProperties}
   {rendererClass=&NiagaraSpriteRendererProperties, rendererIndex=1}
  ]
 }
}
summary {
 systemName=NS_Pickaxe_ClaimReflect_Impact
 emitters [
  {emitterName=LongFlare, rendererClasses [&NiagaraSpriteRendererProperties]}
  {
   emitterName=Flash_Elements
   rendererClasses [&NiagaraSpriteRendererProperties, &NiagaraLightRendererProperties]
  }
  {emitterName=Impact_Smoke, rendererClasses [&NiagaraSpriteRendererProperties]}
  {emitterName=DirectionalFlare, rendererClasses [&NiagaraSpriteRendererProperties]}
  {emitterName=DirectionalStreaks, rendererClasses [&NiagaraSpriteRendererProperties]}
  {emitterName=Sparks, rendererClasses [&NiagaraSpriteRendererProperties]}
  {emitterName=ChromaRadialFlare, rendererClasses [&NiagaraSpriteRendererProperties]}
  {bEnabled=false, emitterName=Shockwave, rendererClasses [&NiagaraMeshRendererProperties]}
 ]
 userVariables [
  User.bIsFrontEnd {type=NiagaraBool}
  User.bIsFrontEndPreview {type=NiagaraBool}
  User.FrontEndScale {type=NiagaraFloat, defaultValue=0.4, description=""}
  User.SwingSectionIndex {type=NiagaraInt32}
  User.Variant {type=NiagaraFloat, defaultValue=0, description=""}
 ]
}
