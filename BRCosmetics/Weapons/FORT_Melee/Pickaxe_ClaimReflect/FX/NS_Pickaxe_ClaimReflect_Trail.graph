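# Niagara swing trail for the ClaimReflect pickaxe, built from four ribbon emitters: main, noise, glow and a tip trail with two renderers.
# Anim-trail modules pick socket pairs and most lifetimes from User.SwingSectionIndex, then fade each ribbon over particle and emitter age.
# A System.Variant index switches between two color looks on selected colors, curves and noise intensity.
# Time-of-day and front-end modifiers adjust opacity, with time-of-day tinting on glow and noise.
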NiagaraSystem /BRCosmetics/Weapons/FORT_Melee/Pickaxe_ClaimReflect/FX/NS_Pickaxe_ClaimReflect_Trail
dependencies {
 usedDataInterfaces [
  &NiagaraDataInterfaceArrayInt32
  &NiagaraDataInterfaceArrayFloat
  &NiagaraDataInterfaceColorCurve
  &NiagaraDataInterfaceCurve
  &NiagaraDataInterfaceSocketReader
 ]
 usedDynamicInputs [&MakeIntFromFloat]
 usedModules [
  &self:SystemSpawnScript.SystemScriptSource.SystemScriptGraph.NiagaraNodeAssignment_0.SetVariables_#1
  &self:SystemSpawnScript.SystemScriptSource.SystemScriptGraph.NiagaraNodeAssignment_1.SetVariables_#2
  &SystemState
 ]
}
emitters {
 GlowTrail {
  emitterUpdateScript [
   &EmitterState()
   &FN_SpawnAnimTrail()
  ]
  particleSpawnScript [
   &FN_InitializeAnimTrailBoneIndecies(
    First Bone or Socket Index:NiagaraInt32=&SelectIntFromArray(Direct Array Index=$User.SwingSectionIndex, Int Selection Array=DataInterface() {propertyValues {IntData [2,0,0,0]}})
    Second Bone or Socket Index:NiagaraInt32=&SelectIntFromArray(Direct Array Index=$User.SwingSectionIndex, Int Selection Array=DataInterface() {propertyValues {IntData [5,4,4,4]}})
   )
   &FN_InitializeAnimTrail(
    AnimTrailWidth=&SelectFloatFromArray(Direct Array Index=$Particles.SpawnGroup, Float Selection Array=DataInterface() {propertyValues {FloatData [0.6,0.04]}})
    Color=&MergeHelper_ColorSwitch(0=<rgba 0.06 0.0733 0.1 1>, 1=<rgba 0.15 0.0075 0.0075 0.6>, 2=?, 3=?, Index=$System.Variant, Num=::NewEnumerator0)
    Lifetime=&SelectFloatFromArray(Direct Array Index=$User.SwingSectionIndex, Float Selection Array=DataInterface() {propertyValues {FloatData [0.2,0.25,0.25,0.25]}})
   )
  ]
  particleUpdateScript [
   &SolveForcesAndVelocity(Clamp Velocity=false, Limit Acceleration=false)
   &FN_UpdateAnimTrail() {module=FN_UpdateAnimTrail001}
   &ScaleColor(
    Scale Alpha=&FloatFromCurve(
     FloatCurve=DataInterface() {propertyValues {Curve {keys [
         {arriveTangent=-5.6915, interpMode=RCIM_Cubic, leaveTangent=-5.6915, tangentMode=RCTM_User, time=0, value=2}
         {interpMode=RCIM_Cubic, tangentMode=RCTM_User, time=1, value=0}
        ]}}}
     Scale Curve=&FloatFromCurve(
      FloatCurve=DataInterface() {propertyValues {Curve {keys [<at 0 1 RCIM_Cubic>,<at 1 0 RCIM_Cubic>]}}}
      CurveIndex=$Emitter.Age
      Scale Curve=&FN_FrontEndFloatModifier(
       Gameplay=&FNTimeOfDayScalar(Night Scale=1, Morning Scale=1, Day Scale=1.5, Evening Scale=1)
       Front End=1
      )
     )
    )
    Scale RGB=&MakeVectorFromLinearColorRGB(Color=&MultiplyLinearColor(
      Linear Color=&ColorFromCurve(ColorCurve=DataInterface() {propertyValues {
         AlphaCurve {keys []}
         BlueCurve {keys [<at 0 5 RCIM_Cubic>,<at 0.212 1>,<at 1 0.1 RCIM_Cubic>]}
         GreenCurve {keys [<at 0 3.6667 RCIM_Cubic>,<at 0.212 1>,<at 1 0.1 RCIM_Cubic>]}
         RedCurve {keys [<at 0 3 RCIM_Cubic>,<at 0.212 1>,<at 1 0.1 RCIM_Cubic>]}
        }})
      Scale Factor=&MakeVector4FromVectorAndFloat(
       "Vector (RGB)"=&MakeVectorFromLinearColorRGB(Color=&FNTimeOfDayColor(Night Color=<rgba 0.48 0.5984 0.8 1>, Morning Color=<rgba 0.6 0.6262 0.8 1>, Day Color=<rgba 1 1 1 1>, Evening Color=<rgba 0.6 0.6 0.6 1>))
       "Float (Alpha)"=1
      )
     ))
    ScaleA=true
    ScaleRGB=true
   )
   &CameraOffset(Camera Offset Amount=-5)
  ]
  rendererClasses [&NiagaraRibbonRendererProperties]
  renderers [{rendererClass=&NiagaraRibbonRendererProperties}]
 }
 MainTrail {
  emitterUpdateScript [
   &EmitterState()
   &FN_SpawnAnimTrail(AnimTrailWidthMode=::FromSecond)
  ]
  particleSpawnScript [
   &FN_InitializeAnimTrailBoneIndecies(
    First Bone or Socket Index:NiagaraInt32=&SelectIntFromArray(Direct Array Index=$User.SwingSectionIndex, Int Selection Array=DataInterface() {propertyValues {IntData [2,0,0,0]}})
    Second Bone or Socket Index:NiagaraInt32=&SelectIntFromArray(Direct Array Index=$User.SwingSectionIndex, Int Selection Array=DataInterface() {propertyValues {IntData [3,1,1,1]}})
   )
   &FN_InitializeAnimTrail(
    AnimTrailWidth=&SelectFloatFromArray(Direct Array Index=$Particles.SpawnGroup, Float Selection Array=DataInterface() {propertyValues {FloatData [0.85,0.04]}})
    Color=&MergeHelper_ColorSwitch(0=<rgba 0.0375 0.0712 0.15 0.3>, 1=<rgba 0.05 0.0108 0.0084 0.1>, 2=?, 3=?, Index=$System.Variant, Num=::NewEnumerator0)
    Lifetime=&SelectFloatFromArray(Direct Array Index=$User.SwingSectionIndex, Float Selection Array=DataInterface() {propertyValues {FloatData [0.2,0.25,0.25,0.25]}})
   )
  ]
  particleUpdateScript [
   &SolveForcesAndVelocity(Clamp Velocity=false, Limit Acceleration=false)
   &FN_UpdateAnimTrail() {module=FN_UpdateAnimTrail001}
   &ScaleColor(
    Scale Alpha=&FloatFromCurve(
     FloatCurve=DataInterface() {propertyValues {Curve {keys [
         {arriveTangent=-2.4237, interpMode=RCIM_Cubic, leaveTangent=-2.4237, tangentMode=RCTM_User, time=0, value=1}
         {arriveTangent=-0.2343, interpMode=RCIM_Cubic, leaveTangent=-0.2343, tangentMode=RCTM_User, time=1, value=0}
        ]}}}
     Scale Curve=&FloatFromCurve(
      FloatCurve=DataInterface() {propertyValues {Curve {keys [<at 0.15 1 RCIM_Cubic>,<at 0.4 0 RCIM_Cubic>]}}}
      CurveIndex=$Emitter.Age
      Scale Curve=&FN_FrontEndFloatModifier(Gameplay=&FNTimeOfDayScalar(Night Scale=0.8, Morning Scale=1, Evening Scale=1), Front End=1)
     )
    )
    Scale RGB=&MakeVectorFromLinearColorRGB(Color=&MultiplyLinearColor(
      Linear Color=&ColorFromCurve(ColorCurve=DataInterface() {propertyValues {
         AlphaCurve {keys []}
         BlueCurve {keys [<at 0 15 RCIM_Cubic>,<at 0.212 0.5>,<at 1 0.05 RCIM_Cubic>]}
         GreenCurve {keys [<at 0 15 RCIM_Cubic>,<at 0.212 0.5>,<at 1 0.05 RCIM_Cubic>]}
         RedCurve {keys [<at 0 15 RCIM_Cubic>,<at 0.212 0.5>,<at 1 0.05 RCIM_Cubic>]}
        }})
      Scale Factor=<1 1 1 1>
     ))
    ScaleA=true
    ScaleRGB=true
   )
  ]
  rendererClasses [&NiagaraRibbonRendererProperties]
  renderers [{rendererClass=&NiagaraRibbonRendererProperties}]
 }
 NoiseTrail {
  emitterUpdateScript [
   &EmitterState()
   &FN_SpawnAnimTrail()
  ]
  particleSpawnScript [
   &FN_InitializeAnimTrailBoneIndecies(
    First Bone or Socket Index:NiagaraInt32=&SelectIntFromArray(Direct Array Index=$User.SwingSectionIndex, Int Selection Array=DataInterface() {propertyValues {IntData [2,0,0,0]}})
    Second Bone or Socket Index:NiagaraInt32=&SelectIntFromArray(Direct Array Index=$User.SwingSectionIndex, Int Selection Array=DataInterface() {propertyValues {IntData [3,1,1,1]}})
   )
   &FN_InitializeAnimTrail(
    AnimTrailWidth=0.98
    Color=<rgba 3 3 3 1>
    Lifetime=&SelectFloatFromArray(Direct Array Index=$User.SwingSectionIndex, Float Selection Array=DataInterface() {propertyValues {FloatData [0.2,0.27,0.27,0.27]}})
   )
  ]
  particleUpdateScript [
   &SolveForcesAndVelocity(Clamp Velocity=false, Limit Acceleration=false)
   &FN_UpdateAnimTrail() {module=FN_UpdateAnimTrail001}
   &ScaleColor(
    Scale Alpha=&FloatFromCurve(
     FloatCurve=DataInterface() {propertyValues {Curve {keys [
         {arriveTangent=-4.7131, interpMode=RCIM_Cubic, leaveTangent=-4.7131, tangentMode=RCTM_User, time=0, value=1.5}
         {arriveTangent=0.0442, interpMode=RCIM_Cubic, leaveTangent=0.0442, tangentMode=RCTM_User, time=1, value=0}
        ]}}}
     Scale Curve=&FloatFromCurve(
      FloatCurve=DataInterface() {propertyValues {Curve {keys [<at 0.15 1 RCIM_Cubic>,<at 0.4 0 RCIM_Cubic>]}}}
      CurveIndex=$Emitter.Age
      Scale Curve=&Multiply_Float(
       A=&FN_FrontEndFloatModifier(
        Gameplay=&FNTimeOfDayScalar(Night Scale=0.7, Morning Scale=1, Day Scale=1.2, Evening Scale=1)
        Front End=1
       )
       B=&MergeHelper_FloatSwitch(0=1, 1=3, 2=?, 3=?, Index=$System.Variant, Num=::NewEnumerator0)
      )
     )
    )
    Scale RGB=&MakeVectorFromLinearColorRGB(Color=&MultiplyLinearColor(
      Linear Color=&MergeHelper_ColorSwitch(
       0=&ColorFromCurve(ColorCurve=DataInterface() {propertyValues {
          AlphaCurve {keys []}
          BlueCurve {keys [<at 0 4 RCIM_Cubic>,<at 0.2178 0.5>,<at 1 0.01 RCIM_Cubic>]}
          GreenCurve {keys [<at 0 2.8 RCIM_Cubic>,<at 0.2178 0.35>,<at 1 0.0082 RCIM_Cubic>]}
          RedCurve {keys [<at 0 2.4 RCIM_Cubic>,<at 0.2178 0.3>,<at 1 0.008 RCIM_Cubic>]}
         }})
       1=&ColorFromCurve(ColorCurve=DataInterface() {propertyValues {
          AlphaCurve {keys []}
          BlueCurve {keys [<at 0 0.0423 RCIM_Cubic>,<at 0.2178 0.0256>,<at 1 0.0026 RCIM_Cubic>]}
          GreenCurve {keys [<at 0 0.075 RCIM_Cubic>,<at 0.2178 0.04>,<at 1 0.0035 RCIM_Cubic>]}
          RedCurve {keys [<at 0 0.2 RCIM_Cubic>,<at 0.2178 0.087>,<at 1 0.01 RCIM_Cubic>]}
         }})
       2=?
       3=?
       Index=$System.Variant
       Num=::NewEnumerator0
      )
      Scale Factor=&MakeVector4FromVectorAndFloat(
       "Vector (RGB)"=&MakeVectorFromLinearColorRGB(Color=&FNTimeOfDayColor(Night Color=<rgba 0.56 0.6488 0.8 1>, Morning Color=<rgba 0.6 0.6262 0.8 1>, Day Color=<rgba 1 1 1 1>, Evening Color=<rgba 0.6 0.6 0.6 1>))
       "Float (Alpha)"=1
      )
     ))
    ScaleA=true
    ScaleRGB=true
   )
  ]
  rendererClasses [&NiagaraRibbonRendererProperties]
  renderers [{rendererClass=&NiagaraRibbonRendererProperties}]
 }
 TipTrail {
  emitterUpdateScript [
   &EmitterState()
   &FN_SpawnAnimTrail(AnimTrailWidthMode=::FromSecond)
  ]
  particleSpawnScript [
   &FN_InitializeAnimTrailBoneIndecies(
    First Bone or Socket Index:NiagaraInt32=&SelectIntFromArray(Direct Array Index=$User.SwingSectionIndex, Int Selection Array=DataInterface() {propertyValues {IntData [2,0,0,0]}})
    Second Bone or Socket Index:NiagaraInt32=&SelectIntFromArray(Direct Array Index=$User.SwingSectionIndex, Int Selection Array=DataInterface() {propertyValues {IntData [7,6,6,6]}})
   )
   &FN_InitializeAnimTrail(
    AnimTrailWidth=0.22
    Color=&MergeHelper_ColorSwitch(0=<rgba 1.3598 1.9119 3 1>, 1=<rgba 4 0 0.6667 0.7>, 2=?, 3=?, Index=$System.Variant, Num=::NewEnumerator0)
    Lifetime=&SelectFloatFromArray(Direct Array Index=$User.SwingSectionIndex, Float Selection Array=DataInterface() {propertyValues {FloatData [0.2,0.25,0.25,0.25]}})
   )
  ]
  particleUpdateScript [
   &FN_UpdateAnimTrail() {module=FN_UpdateAnimTrail001}
   &ScaleColor(
    Scale Alpha=&FloatFromCurve(
     FloatCurve=DataInterface() {propertyValues {Curve {keys [
         {arriveTangent=-3.2786, interpMode=RCIM_Cubic, leaveTangent=-3.2786, tangentMode=RCTM_User, time=0, value=1.5}
         {arriveTangent=-0.4419, interpMode=RCIM_Cubic, leaveTangent=-0.4419, tangentMode=RCTM_User, time=1, value=0}
        ]}}}
     Scale Curve=&FloatFromCurve(
      FloatCurve=DataInterface() {propertyValues {Curve {keys [<at 0.15 1 RCIM_Cubic>,<at 0.4 0 RCIM_Cubic>]}}}
      CurveIndex=$Emitter.Age
      Scale Curve=&FN_FrontEndFloatModifier(
       Gameplay=&FNTimeOfDayScalar(Night Scale=1, Morning Scale=1, Day Scale=1.5, Evening Scale=1)
       Front End=1
      )
     )
    )
    Scale RGB=&MakeVectorFromLinearColorRGB(Color=&MergeHelper_ColorSwitch(
      0=&ColorFromCurve(ColorCurve=DataInterface() {propertyValues {
         AlphaCurve {keys []}
         BlueCurve {keys [<at 0 4 RCIM_Cubic>,<at 0.2938 0.2>,<at 1 0.01 RCIM_Cubic>]}
         GreenCurve {keys [<at 0 2.8 RCIM_Cubic>,<at 0.2938 0.1533>,<at 1 0.0087 RCIM_Cubic>]}
         RedCurve {keys [<at 0 2.4 RCIM_Cubic>,<at 0.2938 0.12>,<at 1 0.008 RCIM_Cubic>]}
        }})
      1=&ColorFromCurve(ColorCurve=DataInterface() {propertyValues {
         AlphaCurve {keys []}
         BlueCurve {keys [<at 0 0 RCIM_Cubic>,<at 0.1904 0.04>,<at 1 0.002 RCIM_Cubic>]}
         GreenCurve {keys [<at 0 0 RCIM_Cubic>,<at 0.1904 0.04>,<at 1 0.002 RCIM_Cubic>]}
         RedCurve {keys [<at 0 5 RCIM_Cubic>,<at 0.1904 0.2>,<at 1 0.01 RCIM_Cubic>]}
        }})
      2=?
      3=?
      Index=$System.Variant
      Num=::NewEnumerator0
     ))
    ScaleA=true
    ScaleRGB=true
   )
  ]
  rendererClasses [&NiagaraRibbonRendererProperties]
  renderers [
   {rendererClass=&NiagaraRibbonRendererProperties}
   {rendererClass=&NiagaraRibbonRendererProperties, rendererIndex=1}
  ]
 }
}
summary {
 systemName=NS_Pickaxe_ClaimReflect_Trail
 emitters [
  {emitterName=MainTrail, rendererClasses [&NiagaraRibbonRendererProperties]}
  {emitterName=NoiseTrail, rendererClasses [&NiagaraRibbonRendererProperties]}
  {emitterName=GlowTrail, rendererClasses [&NiagaraRibbonRendererProperties]}
  {emitterName=TipTrail, rendererClasses [&NiagaraRibbonRendererProperties]}
 ]
 userVariables [
  User.bIsFrontEnd {type=NiagaraBool}
  User.bIsFrontEndPreview {type=NiagaraBool}
  User.RibbonIDOffset {type=NiagaraInt32}
  User.SwingSectionIndex {type=NiagaraInt32}
  User.Variant {type=NiagaraFloat, defaultValue=0, description=""}
 ]
}
