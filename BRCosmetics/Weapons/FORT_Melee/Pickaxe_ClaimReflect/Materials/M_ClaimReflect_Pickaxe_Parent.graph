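# Translucent parent material for the ClaimReflect pickaxe that gives it an animated, glowing plasma look.
# Panning plasma and Voronoi masks, distorted by cloud noise and flow-map UVs, combine with edge masks, flicker and a sine wipe into a grayscale glow.
# Color parameters tint it with an optional second color and rainbow chromatic fringe, and a deploy-wipe function adds a bright wipe.
# Character-parent PreFX and PostFX functions bracket overrides for emissive, black base color and opacity.
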Material /BRCosmetics/Weapons/FORT_Melee/Pickaxe_ClaimReflect/Materials/M_ClaimReflect_Pickaxe_Parent
graph {
 properties {
  bHasPixelAnimation=true
  bUseMaterialAttributes=true
  bUsedWithInstancedStaticMeshes=true
  bUsedWithMeshDeformer=true
  bUsedWithMorphTargets=true
  bUsedWithSkeletalMesh=true
  blendMode=BLEND_Translucent
  pixelDepthOffsetMode=PDOM_Legacy
  previewMesh=/BRCosmetics/Weapons/FORT_Melee/Pickaxe_ClaimReflect/Meshes/ClaimReflect_Axe
  shadingModel=MSM_FromMaterialExpression
  usedShadingModels=Default Lit | Subsurface | Clear Coat | Subsurface Profile | Cloth
 }
 nodes [
  n1 = Multiply(A=ComponentMask(None=NamedRerouteUsage()) {b=true, r=true}) {constB=0.01}
  n2 = NamedRerouteUsage()
  n3 = AppendVector(
   A=ConstantBiasScale(None=MaterialFunctionCall(Float4=@n2) {materialFunction=&BreakOutFloat4Components}.A) {bias=-0.5, scale=2}
   B=Constant()
  )
  n4 = ConstantBiasScale(None=NamedRerouteUsage()) {bias=-0.5, scale=2}
  Add(A=@n1, B=Multiply(A=Multiply(A=@n3, B=@n4)) {constB=0})
  n5 = NamedRerouteUsage()
  n6 = LinearInterpolate(
   A=ScalarParameter() {defaultValue=0.1, parameterName=BGNoiseMin}
   B=ScalarParameter() {defaultValue=1, parameterName=BGNoiseMax}
   Alpha=Max(A=Power(Base=@n5) {constExponent=1}, B=NamedRerouteUsage())
  ) {constB=0.5}
  n7 = Power(Base=OneMinus(None=NamedRerouteUsage())) {constExponent=3}
  n8 = Multiply(A=Power(Base=NamedRerouteUsage()) {constExponent=1}, B=NamedRerouteUsage())
  n9 = ScalarParameter() {defaultValue=15, parameterName=PlasmaBaseMin}
  n10 = ScalarParameter() {defaultValue=30, parameterName=PlasmaBaseMax}
  n11 = NamedRerouteUsage()
  n12 = NamedRerouteUsage()
  n13 = Add(
   A=Multiply(A=Multiply(A=@n7, B=@n8), B=LinearInterpolate(A=@n9, B=@n10, Alpha=@n11)) {constB=50}
   B=Multiply(A=OneMinus(None=@n12), B=ScalarParameter() {defaultValue=1, parameterName=VerticalGradientInt})
  )
  n14 = SmoothStep(Value=Power(Base=NamedRerouteUsage()) {constExponent=5}) {constMax=0.5}
  n15 = Add(A=NamedRerouteUsage(), B=@n14)
  n16 = ScalarParameter() {defaultValue=50, parameterName=BladeEdgeIntensity}
  n17 = Constant() {r=0.06}
  n18 = Subtract(
   A=Subtract(A=NamedRerouteUsage()) {constB=0.2}
   B=Multiply(A=MaterialFunctionCall!(FlickerInitialSpeed=@n17, FlickerRandom=@n17) {materialFunction=/Game/Effects/Fort_Effects/Materials/_Functions/MF_Flicker}.Result) {constB=0.2}
  )
  n19 = NamedRerouteUsage()
  n20 = Multiply(A=Max(A=NamedRerouteUsage(), B=Power(Base=@n18) {constExponent=4}), B=Multiply(A=@n19, B=@n19))
  n21 = NamedRerouteUsage()
  n22 = LinearInterpolate(Alpha=@n21) {constA=-0.5}
  n23 = NamedRerouteUsage()
  n24 = Multiply(A=Multiply(A=Add(A=@n22, B=@n23), B=@n23), B=ScalarParameter() {defaultValue=20, parameterName=EdgeCornersInt}) {constB=10}
  n25 = Add(A=Add(A=Add(A=@n6, B=@n13), B=Multiply(A=@n15, B=@n14)), B=Add(A=LinearInterpolate(B=@n16, Alpha=@n20), B=@n24))
  n26 = LinearInterpolate(B=ScalarParameter() {defaultValue=1.5, parameterName=WipeIntensity}, Alpha=NamedRerouteUsage()) {constA=1}
  n27 = Multiply(
   A=VectorParameter() {parameterName=MainColor, defaultValue=<rgba 0.1337 0.2092 0.5833 1>}.RGB
   B=NamedRerouteDeclaration(None=Multiply(A=@n25, B=@n26)) {name=GrayscaleMap, nodeColor=<rgba 0.3594 1 0 1>}
  )
  n28 = VectorParameter() {parameterName=SecondaryColor, defaultValue=<rgba 50 0.9959 0 1>}
  Add(A=@n27, B=Multiply(A=@n28.RGB))
  n29 = MaterialFunctionCall() {materialFunction=&MF_CharacterParent_2023_PreFX}
  MaterialFunctionCall(Material=NamedRerouteUsage(), M=@n29.M, Pre Skinned Local Position=@n29.M, Distance Blend=@n29.M, Fresnel=@n29.M) {materialFunction=&MF_CharacterParent_2023_PostFX}
  n30 = TextureCoordinate() {coordinateIndex=1}
  n31 = TextureSample(UVs=@n30) {samplerType=SAMPLERTYPE_LinearColor, texture=&/T_Pickaxe_ClaimReflect_FlowMap}
  n32 = ConstantBiasScale(None=NamedRerouteUsage()) {bias=-0.5, scale=2}
  n33 = Add(A=@n31.R, B=Multiply(A=Multiply(A=@n31.G, B=@n31.G), B=@n32))
  Multiply(A=Multiply(A=AppendVector(A=@n33, B=@n31.G), B=Constant2Vector() {g=2.5, r=1})) {constB=0.1}
  n34 = NamedRerouteUsage()
  Multiply(A=@n34, B=@n34)
  n35 = NamedRerouteUsage()
  n36 = Multiply(A=@n35)
  n37 = TextureSample(UVs=@n36) {texture=&AtmosphericCloudNoise01}
  Multiply(A=@n37.B)
  n38 = Add(A=LocalPosition().XYZ, B=Constant3Vector())
  n39 = NamedRerouteDeclaration(None=VertexInterpolator(VS=@n38).PS) {name=PreSkinnedPos, nodeColor=<rgba 0.5703 1 0 1>}
  NamedRerouteDeclaration(None=Saturate(None=ComponentMask(None=@n39) {b=true})) {name=BladeSideMask, nodeColor=<rgba 1 0 0.4453 1>}
  n40 = TextureCoordinate() {coordinateIndex=1}
  n41 = TextureSample(UVs=TextureCoordinate() {coordinateIndex=1}) {samplerType=SAMPLERTYPE_LinearColor, texture=&T_Pickaxe_ClaimReflect_FlowMap}
  n42 = ComponentMask(None=@n41.RGB) {g=true, r=true}
  n43 = Multiply(A=Multiply(A=@n32, B=ConstantBiasScale(None=@n42) {bias=-0.5, scale=2})) {constB=0.15}
  NamedRerouteDeclaration(None=Multiply(A=Add(A=@n40, B=@n43))) {name=UVs, nodeColor=<rgba 0 0.7656 1 1>}
  NamedRerouteDeclaration(None=@n29.Result) {name=InMaterial, nodeColor=<rgba 0.9453 1 0 1>}
  n44 = Constant()
  n45 = ScalarParameter() {defaultValue=0.2, parameterName=MinOpacity}
  n46 = Saturate(None=NamedRerouteUsage())
  n47 = SetMaterialAttributes!(
   MaterialAttributes=NamedRerouteUsage()
   Emissive Color=NamedRerouteUsage()
   Base Color=@n44
   Roughness=Constant() {r=1}
   Opacity=Multiply(A=NamedRerouteUsage(), B=LinearInterpolate(A=@n45, Alpha=@n46))
   Specular=@n44
  ) {attributeSetTypes [
    #1
    #2
    #3
    #4
    #5
   ]}
  NamedRerouteDeclaration(None=@n47) {name=OutMaterial, nodeColor=<rgba 0.7344 0 1 1>}
  NamedRerouteDeclaration(None=MaterialFunctionCall(Float2=@n30) {materialFunction=&BreakOutFloat2Components}.G) {name=VerticalGradient, nodeColor=<rgba 0 1 0.9063 1>}
  n48 = Constant2Vector() {g=1, r=1}
  n49 = Subtract(A=@n37.R) {constB=0.5}
  n50 = Add(A=Multiply(A=Multiply(A=@n35, B=@n48) {constB=1.1111111318516531e+18}) {constB=3}, B=Multiply(A=@n49) {constB=0.15})
  NamedRerouteDeclaration(None=TextureSample(UVs=Panner(Coordinate=@n50) {speedY=1}) {texture=&T_Tile_RainbowColor}.RGB) {name=OutChromatic, nodeColor=<rgba 1 0 0.1641 1>}
  n51 = ConstantBiasScale(None=TextureSample() {samplerType=SAMPLERTYPE_Masks, texture=&/ClaimReflect_Sword_curve}.G) {bias=-0.5, scale=2}
  NamedRerouteDeclaration(None=Saturate(None=Subtract(A=@n51) {constB=0.01})) {name=Edges, nodeColor=<rgba 1 0 0 1>}
  n52 = Constant3Vector() {constant=<rgba 0 1 0 0>}
  n53 = PreSkinnedNormal()
  n54 = OneMinus(None=Abs(None=DotProduct(A=@n52, B=@n53)))
  NamedRerouteDeclaration(None=VertexInterpolator(VS=Saturate(None=@n54)).PS) {name=BladeEdge, nodeColor=<rgba 0.3359 1 0 1>}
  NamedRerouteDeclaration(None=@n31.RGBA) {name=UVMap, nodeColor=<rgba 0 1 0.9531 1>}
  n55 = Multiply(A=ComponentMask(None=NamedRerouteUsage()) {r=true}) {constB=0.01}
  n56 = Multiply(A=Time()) {constB=1.5}
  n57 = ConstantBiasScale(None=Sine(None=Add(A=@n55, B=@n56)))
  NamedRerouteDeclaration(None=Multiply(A=Multiply(A=@n57, B=@n57), B=NamedRerouteUsage())) {name=Wipe, nodeColor=<rgba 0.8047 1 0 1>}
  NamedRerouteDeclaration(None=@n31.B) {name=EdgeGradient, nodeColor=<rgba 1 0.9141 0 1>}
  n58 = MaterialFunctionCall!(InDerectionalGradient=NamedRerouteUsage()) {materialFunction=&/MF_ClaimReflect_DeployWipe}
  NamedRerouteDeclaration(None=@n58.WipeMask) {name=WipeMask, nodeColor=<rgba 1 0 0 1>}
  NamedRerouteDeclaration(None=@n58.WipeLine) {name=WipeLine, nodeColor=<rgba 0 1 1 1>}
  NamedRerouteDeclaration(None=@n8) {name=Plasma, nodeColor=<rgba 1 0 0.2344 1>}
  NamedRerouteDeclaration(None=@n24) {name=EnergyEdge, nodeColor=<rgba 0 0.5547 1 1>}
  NamedRerouteDeclaration(None=Step(X=NamedRerouteUsage()) {constY=0.6}) {name=EdgeMask, nodeColor=<rgba 0.5938 0 1 1>}
  n59 = Multiply(A=@n35) {constB=1.75}
  n60 = Multiply(A=@n49) {constB=0.075}
  n61 = TextureSample(UVs=Add(A=Panner(Coordinate=@n59) {speedY=0.5}, B=Multiply(A=@n60) {constB=-1})) {samplerType=SAMPLERTYPE_Masks, texture=&/T_Pickaxe_ClaimReflect_Mask}
  NamedRerouteDeclaration(None=@n61.R) {name=LowFreqPlasma, nodeColor=<rgba 0 1 0.2734 1>}
  NamedRerouteDeclaration(None=@n61.G) {name=LowFreqVoronoi, nodeColor=<rgba 1 0 0.0469 1>}
  n62 = Multiply(A=Subtract(A=TextureSample(UVs=@n1) {texture=&AtmosphericCloudNoise01}.G) {constB=0.5}) {constB=0.1}
  n63 = NamedRerouteUsage()
  n64 = Panner(Coordinate=Add(A=Add(A=@n1, B=@n62), B=Multiply(A=@n63) {constB=0.2})) {speedX=0.1}
  n65 = TextureSample(UVs=@n64) {samplerType=SAMPLERTYPE_Masks, texture=&T_Pickaxe_ClaimReflect_Mask}
  NamedRerouteDeclaration(None=@n65.R) {name=MidFreqPlasma, nodeColor=<rgba 0 0.3672 1 1>}
  NamedRerouteDeclaration(None=@n65.G) {name=MidFreqVoronoi, nodeColor=<rgba 0.6875 1 0 1>}
  n66 = Multiply(A=@n38) {constB=0.01}
  n67 = Constant()
  NamedRerouteDeclaration(None=Power(Base=SphereMask(A=@n66, B=@n67) {attenuationRadius=0.25, hardnessPercent=0}) {constExponent=3}) {name=BaseRadialMask, nodeColor=<rgba 0.9922 0 1 1>}
  n68 = NamedRerouteUsage()
  n69 = ScalarParameter() {defaultValue=2.5317, parameterName=ExtraPlasmaExp}
  n70 = SmoothStep(Value=Power(Base=NamedRerouteUsage(), Exp=ScalarParameter() {defaultValue=2.1404, parameterName=ExtraEdgeExp}) {constExponent=5}) {constMax=0.7, constMin=0.1}
  n71 = Max(A=NamedRerouteUsage(), B=NamedRerouteUsage())
  n72 = Multiply(A=@n70, B=Add(A=@n70, B=LinearInterpolate(Alpha=@n71) {constA=-1}))
  n73 = Add(A=@n72, B=NamedRerouteUsage())
  n74 = Max(
   A=Multiply(A=Power(Base=@n68, Exp=@n69) {constExponent=3}, B=ScalarParameter() {defaultValue=5, parameterName=ExtraPlasmaInt}) {constB=50}
   B=Multiply(A=Multiply(A=@n72, B=@n73), B=ScalarParameter() {defaultValue=1, parameterName=ExtraEdgeInt})
  )
  n75 = Multiply(A=NamedRerouteUsage(), B=LinearInterpolate(Alpha=NamedRerouteUsage()) {constA=0.5})
  n76 = Power(Base=OneMinus(None=NamedRerouteUsage()))
  n77 = Add(
   A=LinearInterpolate(A=@n27, B=LinearInterpolate(A=@n27, B=@n28.RGB, Alpha=@n74), Alpha=ScalarParameter() {parameterName=UseTwoColors})
   B=Multiply(A=Multiply(A=@n75, B=@n76), B=ScalarParameter() {defaultValue=0.5, parameterName=ChromAberrIntensity}) {constB=0.5}
  )
  n78 = Add(A=Multiply(A=NamedRerouteUsage(), B=ScalarParameter() {defaultValue=10, parameterName=DeployWipeInt}) {constB=10}, B=NamedRerouteUsage())
  n79 = ScalarParameter() {defaultValue=-0.5, parameterName=WipeColorSaturationAmt}
  n80 = NamedRerouteUsage()
  NamedRerouteDeclaration(None=Desaturation(None=Multiply(A=@n77, B=@n78), Fraction=LinearInterpolate(B=@n79, Alpha=@n80))) {name=OutEmissive, nodeColor=<rgba 0 0.625 1 1>}
  NamedRerouteUsage()
  OneMinus(None=@n21)
  Panner(Coordinate=Add(A=Multiply(A=@n59) {constB=0.5}, B=@n60)) {speedY=0.2}
  Power(Base=@n65.R) {constExponent=3}
  TextureSample(UVs=@n36) {texture=/BRCosmetics/Athena/Effects/Particles/Skydiving/NS_Contrail_PinkSpike/Textures/T_PinkSpike_Voronoi2D_Height}
  TextureSample(UVs=@n64) {samplerType=SAMPLERTYPE_Masks, texture=&T_Pickaxe_ClaimReflect_Mask}
 ]
}
